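# Niagara aura effect for a weapon cache chest, with size variants picked by the FX_WeaponCaseSize index.
# Aura burst-spawns a persistent faint mesh whose width follows that index.
# Corona puts two blue sprite-and-mesh glows at the case's ends, pulsing alpha by waveform and fading each by a horizontal camera-direction dot product; FXScale scales their meshes.
# A SmallParticles emitter with curl-noise motes exists but is disabled.
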NiagaraSystem /PaprikaWeaponCases/Assets/FX/NS_WeaponCacheChest_Aura
dependencies {
 usedDataInterfaces [
  &NiagaraDataInterfaceArrayFloat
  &NiagaraDataInterfaceCamera
  &NiagaraDataInterfaceCurve
  &NiagaraDataInterfaceColorCurve
 ]
 usedDynamicInputs []
 usedModules [&SystemState]
}
emitters {
 Aura {
  emitterUpdateScript [
   &EmitterState(Life Cycle Mode=::NewEnumerator1, Loop Behavior=::NewEnumerator1, Loop Duration Mode=::NewEnumerator1, UseLoopDelay=false)
   &SpawnBurst_Instantaneous(Use Spawn Probability=false, UseLoopCountLimit=true)
  ]
  particleSpawnScript [
   &InitializeParticle(
    Color=<rgba 1 1 1 0.045>
    Mesh Scale=&MakeVector(
     X=&SelectFloatFromArray(Direct Array Index=$User.FX_WeaponCaseSize, Float Selection Array=DataInterface() {propertyValues {FloatData [0.775,1,1.335]}})
     Y=2
     Z=0.9
    )
    Mesh Scale Mode=::NewEnumerator4
    Position Offset=<0 -3.5 0>
    UsePositionOffset=true
    Write Material Random=true
   )
   &InitialMeshOrientation(Mesh Orientation Mode=::NewEnumerator2, Use Rotation Vector=false) {enabled=false}
  ]
  particleUpdateScript [
   &ParticleState(Kill Particles When Lifetime Has Elapsed=false, Let Infinitely Lived Particles Die When Emitter Deactivates=true)
   &DynamicMaterialParameters(
    Index 0 Param 1=1
    Index 0 Param 2=0.5
    Index 0 Param 3=5
    Module.Index0Param0Write=true
    Module.Index0Param1Write=true
    Module.Index0Param2Write=true
    Module.Index0Param3Write=false
   )
  ]
  rendererClasses [&NiagaraMeshRendererProperties]
  renderers [{rendererClass=&NiagaraMeshRendererProperties, rendererIndex=0}]
 }
 Corona {
  emitterUpdateScript [
   &EmitterState(Life Cycle Mode=::NewEnumerator1, Loop Behavior=::NewEnumerator1, Loop Duration Mode=::NewEnumerator1, UseLoopDelay=false)
   &SpawnBurst_Instantaneous(Spawn Count=2, Use Spawn Probability=false, UseLoopCountLimit=true)
   /Niagara/Noise/Waveform(
    EnableExponent1=false
    Global Input Position:NiagaraFloat=&RandomRangeFloat(Minimum=0, Maximum=0.1)
    GlobalClampMax=false
    GlobalClampMin=false
    UseInput1=false
    UsePhase1=false
    "[1] Frequency":NiagaraFloat=&RandomRangeFloat(Minimum=0.15, Maximum=0.25)
   ) {module=Waveform001}
  ]
  particleSpawnScript [
   &PartitionParticles()
   &InitializeParticle(
    Color=<rgba 0.1 0.325 1 1>
    Mesh Scale Mode=::NewEnumerator3
    Mesh Uniform Scale=&Multiply_Float(A=3, B=$User.FXScale)
    Position=&MakeVector(
     X=&Multiply_Float(
      A=&Lerp_Float(
       A=&SelectFloatFromArray(Direct Array Index=$User.FX_WeaponCaseSize, Float Selection Array=DataInterface() {propertyValues {FloatData [-45.5,-60,-80.5]}})
       B=&SelectFloatFromArray(Direct Array Index=$User.FX_WeaponCaseSize, Float Selection Array=DataInterface() {propertyValues {FloatData [45.5,60,80.5]}})
       Alpha=&MakeFloatFromInt(Int=$Particles.PartitionParticles.Partition)
      )
      B=1
     )
     Y=-3
     Z=0
    )
    Position Mode=::NewEnumerator1
    Sprite Size=<40 90>
    Sprite Size Mode=::NewEnumerator4
    UsePositionOffset=false
    Write Material Random=true
   )
   /SetVariables(Particles.SpriteFlareSize:Vector2f=&Multiply_Vector2D(A=$Particles.SpriteSize, B=<0.38 1.3>)) {enabled=false, isSetParameters=true}
   &SpriteFacingAndAlignment(Sprite Alignment=<0 0 1>, Sprite Facing=<0 1 0>, WriteAlignment=true, WriteFacing=false)
  ]
  particleUpdateScript [
   &ParticleState(Kill Particles When Lifetime Has Elapsed=false, Let Infinitely Lived Particles Die When Emitter Deactivates=true)
   &UpdateMeshOrientation(
    Facing Direction=&ReturnCameraVector(Camera Vectors=::NewEnumerator4, Coordinate Space=::Simulation, Origin Position=$Particles.Position)
    Facing Interpolation=false
    Orientation Method=::NewEnumerator1
    Override Quat=false
    Stabilization Mode:ENiagara_OrientVectorCount=::NewEnumerator0
    TransformFacing=false
   ) {enabled=false}
   /SetVariables(Particles.FacingCamera:NiagaraFloat=&ClampFloat(Float=&RemapRange(
      Input Value To Remap=&DotProduct(
       Vector A=&NormalizeVector(
        Vector To Normalize=&Multiply_Vector(
         A=&ReturnCameraVector(Camera Vectors=::NewEnumerator4, Coordinate Space=::Simulation, Origin Position=$Particles.Position)
         B=<1 1 0>
        )
        Safe Normalize=false
        Safe Fallback Threshold=0
        Safe Fallback Vector=<0 0 0.001>
       )
       Vector B=&Lerp_Vector(A=<0.5 0.5 0>, B=<-0.5 0.5 0>, Alpha=&MakeFloatFromInt(Int=$Particles.PartitionParticles.Partition))
      )
      Input Min=0.25
      Input Max=0.6
      Output Min=0
      Output Max=1
     ))) {isSetParameters=true}
   &CameraOffset(Camera Offset Amount=30)
   &DynamicMaterialParameters(
    Index 0 Param 1=1
    Index 1 Param 1=5
    Module.Index0Param0Write=true
    Module.Index0Param1Write=false
    Module.Index0Param2Write=false
    Module.Index0Param3Write=false
    Module.Index1Param0Write=true
    Module.Index1Param1Write=false
    Module.Index1Param2Write=false
    Module.Index1Param3Write=false
    Write Parameter Index 1=true
   ) {enabled=false, module=DynamicMaterialParameters001}
   &ScaleColor(
    Scale Alpha=&Multiply_Float(A=$Particles.FacingCamera, B=&Lerp_Float(A=0.25, B=0.6, Alpha=$Emitter.Waveform001.Waveform))
    ScaleA=true
    ScaleRGB=true
   ) {module=ScaleColor001}
   /SetVariables(Particles.ColorFlare:LinearColor=&MakeLinearColorFromVectorAndFloat(
     "Vector (RGB)"=&MakeVectorFromLinearColorRGB(Color=$Particles.Color)
     "Float (Alpha)"=&Lerp_Float(A=0.3, B=0.5, Alpha=$Emitter.Waveform001.Waveform)
    )) {enabled=false, isSetParameters=true}
  ]
  rendererClasses [&NiagaraSpriteRendererProperties, &NiagaraMeshRendererProperties]
  renderers [
   {rendererClass=&NiagaraSpriteRendererProperties}
   {rendererClass=&NiagaraSpriteRendererProperties, rendererIndex=1}
   {rendererClass=&NiagaraSpriteRendererProperties, rendererIndex=2}
   {rendererClass=&NiagaraMeshRendererProperties, rendererIndex=3}
  ]
 }
 SmallParticles {
  bEnabled=false
  emitterUpdateScript [
   &EmitterState(Life Cycle Mode=::NewEnumerator1, Loop Behavior=::NewEnumerator1, Loop Duration Mode=::NewEnumerator1, UseLoopDelay=false)
   &SpawnRate(SpawnRate=20, Use Spawn Probability=false)
  ]
  particleSpawnScript [
   &InitializeParticle(
    Color=<rgba 3 3 3 0.5>
    Lifetime Max=1.5
    Lifetime Mode=::NewEnumerator1
    Sprite Size Max=<0.7 3>
    Sprite Size Min=<0.7 3>
    Sprite Size Mode=::NewEnumerator2
    UsePositionOffset=false
    Write Material Random=true
   )
   &PartitionParticles() {enabled=false}
   &ShapeLocation(
    Box Size=&MakeVector(
     X=&SelectFloatFromArray(Direct Array Index=$User.FX_WeaponCaseSize, Float Selection Array=DataInterface() {propertyValues {FloatData [80,110,150]}})
     Y=20
     Z=65
    )
    MaskSpawnGroup=false
    Offset Mode=::NewEnumerator0
    Override Seed=false
    Shape Primitive=::NewEnumerator2
    Surface Only=false
    UseOwnerScale=true
   ) {enabled=false}
   &KillParticlesInVolume(
    Box Size:Vector3f=&Multiply_VectorByFloat(
     Vector=&MakeVector(
      X=&SelectFloatFromArray(Direct Array Index=$User.FX_WeaponCaseSize, Float Selection Array=DataInterface() {propertyValues {FloatData [80,110,150]}})
      Y=30
      Z=65
     )
     Float=0.8
    )
    Kill Shape=::NewEnumerator1
   ) {enabled=false}
   /SetVariables(Particles.Position=&Multiply_Vector(
     A=Linked() {linkedVariable {name=Particles.Position, type=NiagaraPosition}}
     B=&MakeVector(
      X=&SelectFloatFromArray(Direct Array Index=$User.FX_WeaponCaseSize, Float Selection Array=DataInterface() {propertyValues {FloatData [0.73,1,1.45]}})
      Y=0.5
      Z=1
     )
    )) {isSetParameters=true}
   &AddVelocity(Module.Use Falloff Distance=false, Origin Offset=<0 -100 0>, Override Seed=false, Velocity Mode=::NewEnumerator1, Velocity Speed=20) {module=AddVelocity001}
   &AddVelocity(Rotation Coordinate Space=::World, Velocity=<0 1 0>, Velocity Speed Scale=28.465) {enabled=false}
  ]
  particleUpdateScript [
   &ParticleState()
   &ScaleSpriteSize(
    Scale Factor=&MakeVector2D(
     X=1
     Y=&FloatFromCurve(FloatCurve=DataInterface() {propertyValues {Curve {keys [
          {tangentMode=RCTM_User, time=0, value=1}
          {arriveTangent=-0.0405, interpMode=RCIM_Cubic, leaveTangent=-0.0405, tangentMode=RCTM_User, time=1.9, value=1}
          {arriveTangent=-8.0089, interpMode=RCIM_Cubic, leaveTangent=-8.0089, tangentMode=RCTM_User, time=2, value=0}
         ]}}})
    )
    Scale Sprite Size Mode=::NewEnumerator2
   )
   &CurlNoiseForce(Noise Frequency=24, Noise Strength=36, Pan Noise=true, Pan Noise Field=<0 0 5>)
   &SolveForcesAndVelocity(Clamp Velocity=false, Limit Acceleration=false)
   &ScaleColor(
    Scale Alpha=&FloatFromCurve(FloatCurve=DataInterface() {propertyValues {Curve {keys [
         {arriveTangent=2.2032, interpMode=RCIM_Cubic, leaveTangent=2.2032, tangentMode=RCTM_User, time=0, value=0}
         {arriveTangent=-2.5249, interpMode=RCIM_Cubic, leaveTangent=-2.5249, tangentMode=RCTM_User, time=0.1408, value=0.9847}
         {arriveTangent=-0.0405, interpMode=RCIM_Cubic, leaveTangent=-0.0405, tangentMode=RCTM_User, time=1, value=0}
        ]}}})
    Scale RGB=&MakeVectorFromLinearColorRGB(Color=&ColorFromCurve(ColorCurve=DataInterface() {propertyValues {
        AlphaCurve {keys []}
        BlueCurve {keys [<at 0.0619 1 RCIM_Cubic>,<at 0.6287 1 RCIM_Cubic>]}
        GreenCurve {keys [<at 0.0619 0.7333 RCIM_Cubic>,<at 0.6287 0.2 RCIM_Cubic>]}
        RedCurve {keys [<at 0.0619 0.2 RCIM_Cubic>,<at 0.6287 0.2 RCIM_Cubic>]}
       }}))
    ScaleA=true
    ScaleRGB=true
   )
   &ScaleColor(
    Scale Alpha=$System.TODAlpha
    Scale RGB=&VectorFromFloat(Value=$System.TODBrightness)
    ScaleA=true
    ScaleRGB=true
   ) {enabled=false, module=ScaleColor001}
  ]
  rendererClasses [&NiagaraSpriteRendererProperties]
  renderers [{rendererClass=&NiagaraSpriteRendererProperties}]
 }
}
summary {
 systemName=NS_WeaponCacheChest_Aura
 emitters [
  {emitterName=Corona, rendererClasses [&NiagaraSpriteRendererProperties, &NiagaraMeshRendererProperties]}
  {bEnabled=false, emitterName=SmallParticles, rendererClasses [&NiagaraSpriteRendererProperties]}
  {emitterName=Aura, rendererClasses [&NiagaraMeshRendererProperties]}
 ]
 userVariables [
  User.FX_WeaponCaseSize {type=NiagaraInt32, defaultValue=1}
  User.FXScale {type=NiagaraFloat, defaultValue=1, description=""}
 ]
}
