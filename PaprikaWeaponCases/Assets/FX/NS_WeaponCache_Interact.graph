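# Niagara system for a short blue burst effect, named for interacting with a weapon cache.
# An Aura mesh, paired end Flares, a large glow with light, smoke puffs and falling sparks play as instantaneous bursts.
# User.FX_WeaponCaseSize indexes three-entry float arrays that fit widths, flare offsets and spawn shapes to the case size.
# Curves drive brief overbright fades and scaling, while a time-of-day scalar adjusts the light color's alpha.
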NiagaraSystem /PaprikaWeaponCases/Assets/FX/NS_WeaponCache_Interact
dependencies {
 usedDataInterfaces [
  &NiagaraDataInterfaceArrayFloat
  &NiagaraDataInterfaceCurve
  &NiagaraDataInterfaceVector2DCurve
  &NiagaraDataInterfaceColorCurve
 ]
 usedDynamicInputs [&Multiply_VectorByFloat]
 usedModules [
  &self:SystemSpawnScript.SystemScriptSource.SystemScriptGraph.NiagaraNodeAssignment_0.SetVariables_#1
  &self:SystemSpawnScript.SystemScriptSource.SystemScriptGraph.NiagaraNodeAssignment_2.SetVariables_#2
  &SystemState
 ]
}
emitters {
 Aura {
  emitterUpdateScript [
   &EmitterState(Life Cycle Mode=::NewEnumerator1, Loop Behavior=::NewEnumerator1, Loop Delay=0.01, Loop Duration Mode=::NewEnumerator1, UseLoopDelay=true)
   &SpawnBurst_Instantaneous(Use Spawn Probability=false, UseLoopCountLimit=true)
  ]
  particleSpawnScript [
   &InitializeParticle(
    Color=<rgba 0.7 0.75 1 0.2>
    Lifetime=0.22
    Mesh Scale=&MakeVector(
     X=&SelectFloatFromArray(Direct Array Index=$User.FX_WeaponCaseSize, Float Selection Array=DataInterface() {propertyValues {FloatData [0.775,1,1.335]}})
     Y=1
     Z=0.85
    )
    Mesh Scale Mode=::NewEnumerator4
    Position Offset=<0 -3.5 0>
    UsePositionOffset=true
    Write Material Random=true
   )
   &InitializeParticle(
    Color=<rgba 0.4 1 0.6 1>
    Lifetime=0.25
    Mesh Scale=&MakeVector(
     X=&SelectFloatFromArray(Direct Array Index=$User.FX_WeaponCaseSize, Float Selection Array=DataInterface() {propertyValues {FloatData [0.92,1.16,1.56]}})
     Y=2.5
     Z=1.1
    )
    Mesh Scale Mode=::NewEnumerator4
    Position Offset=<0 -7.5 0>
    UsePositionOffset=true
    Write Material Random=true
   ) {enabled=false, module=InitializeParticle001}
   &InitialMeshOrientation(Mesh Orientation Mode=::NewEnumerator2, Use Rotation Vector=false) {enabled=false}
  ]
  particleUpdateScript [
   &ParticleState()
   &ScaleMeshSize(Scale Factor=&MakeVector(
     X=1
     Y=&FloatFromCurve(FloatCurve=DataInterface() {propertyValues {Curve {keys [
          {arriveTangent=0.3448, interpMode=RCIM_Cubic, leaveTangent=0.3448, tangentMode=RCTM_User, time=0, value=0.8}
          {arriveTangent=0.0383, interpMode=RCIM_Cubic, leaveTangent=0.0383, tangentMode=RCTM_User, time=0.2, value=2}
         ]}}})
     Z=1
    ))
   &ScaleColor(
    Scale Alpha=&FloatFromCurve(FloatCurve=DataInterface() {propertyValues {Curve {keys [
         {arriveTangent=-1.2905, interpMode=RCIM_Cubic, leaveTangent=-1.2905, tangentMode=RCTM_User, time=0, value=1}
         {interpMode=RCIM_Cubic, tangentMode=RCTM_User, time=1, value=0}
        ]}}})
    Scale RGB=&VectorFromFloat(Value=&FloatFromCurve(FloatCurve=DataInterface() {propertyValues {Curve {keys [{interpMode=RCIM_Cubic, tangentMode=RCTM_User, time=0.2, value=4}, {interpMode=RCIM_Cubic, tangentMode=RCTM_User, time=0.35, value=1}]}}}))
    ScaleA=true
    ScaleRGB=true
   )
   &DynamicMaterialParameters(
    Index 0 Param 1=1
    Index 0 Param 2=1
    Index 0 Param 3=0.1
    Module.Index0Param0Write=true
    Module.Index0Param1Write=true
    Module.Index0Param2Write=true
    Module.Index0Param3Write=false
   )
  ]
  rendererClasses [&NiagaraMeshRendererProperties]
  renderers [{rendererClass=&NiagaraMeshRendererProperties, rendererIndex=0}]
 }
 Flares {
  emitterUpdateScript [
   &EmitterState(Life Cycle Mode=::NewEnumerator1, Loop Behavior=::NewEnumerator1, Loop Delay=0.02, Loop Duration=0.5, UseLoopDelay=false)
   &SpawnBurst_Instantaneous(Spawn Count=2, Use Spawn Probability=false, UseLoopCountLimit=false)
  ]
  particleSpawnScript [
   &PartitionParticles()
   &InitializeParticle(
    Color Mode=::NewEnumerator0
    Lifetime=0.35
    Position Offset=<0 0 5>
    Sprite Size=<67 120>
    Sprite Size Mode=::NewEnumerator4
    UsePositionOffset=false
    Write Material Random=true
   )
   &CameraOffset(Camera Offset Amount=28)
   /SetVariables(Particles.Position=&AddVectorToPosition(
     Position=$System.FX_Position
     Vector=&MakeVector(
      X=&MaskFloatBySpawnGroup(
       Masked Float=&SelectFloatFromArray(Direct Array Index=$User.FX_WeaponCaseSize, Float Selection Array=DataInterface() {propertyValues {FloatData [-45,-59,-80]}})
       Passthrough Float=&SelectFloatFromArray(Direct Array Index=$User.FX_WeaponCaseSize, Float Selection Array=DataInterface() {propertyValues {FloatData [45,59,80]}})
       Spawn Group Mask=$Particles.PartitionParticles.Partition
      )
      Y=-2
      Z=0
     )
    )) {isSetParameters=true}
   &SpriteFacingAndAlignment(
    Alignment Coordinate Space=::Local
    Facing Coordinate Space=::Local
    Sprite Alignment=<0 0 1>
    Sprite Facing=<0 0 1>
    WriteAlignment=true
    WriteFacing=true
   )
   /SetVariables(Particles.Color=&MakeLinearColorFromVectorAndFloat(
     "Vector (RGB)"=&MakeVectorFromLinearColorRGB(Color=<rgba 0.01 0.241 1 1>)
     "Float (Alpha)"=&FNTimeOfDayScalar(Night Scale=1, Morning Scale=1, Evening Scale=1)
    )) {isSetParameters=true}
   /SetVariables(Particles.ColorDarkGlow:LinearColor=<rgba 0 0 0 1>) {enabled=false, isSetParameters=true}
   /SetVariables(Particles.SpriteSizeCorona:Vector2f=<70 70>) {isSetParameters=true}
  ]
  particleUpdateScript [
   &ParticleState()
   &ScaleSpriteSize(
    Non-Uniform Curve Sprite Scale=DataInterface() {propertyValues {
      XCurve {keys [
        {arriveTangent=0.0425, interpMode=RCIM_Cubic, leaveTangent=0.0425, tangentMode=RCTM_User, time=0.1, value=0.7}
        {arriveTangent=0.0084, interpMode=RCIM_Cubic, leaveTangent=0.0084, tangentMode=RCTM_User, time=0.26, value=1}
        <at 0.8 0.7>
       ]}
      YCurve {keys [
        {arriveTangent=0.056, interpMode=RCIM_Cubic, leaveTangent=0.056, tangentMode=RCTM_User, time=0.35, value=1.5}
        {arriveTangent=2.5472, interpMode=RCIM_Cubic, leaveTangent=2.5472, tangentMode=RCTM_User, time=0.8, value=2}
       ]}
     }}
    Scale Sprite Size Mode=::NewEnumerator3
   ) {module=ScaleSpriteSize001}
   /SetVariables(Particles.SpriteSizeDarkGlow:Vector2f=&Multiply_Vector2D_ByFloat(Float=85, Vector2D=<1 1.2>)) {isSetParameters=true}
   &ScaleSpriteSize(
    Non-Uniform Curve Sprite Scale=DataInterface() {propertyValues {
      XCurve {keys [
        {arriveTangent=3.4417, interpMode=RCIM_Cubic, leaveTangent=3.4417, tangentMode=RCTM_User, time=0.6, value=1}
        {arriveTangent=2.6456, interpMode=RCIM_Cubic, leaveTangent=2.6456, tangentMode=RCTM_User, time=1, value=4}
       ]}
      YCurve {keys [
        {interpMode=RCIM_Cubic, tangentMode=RCTM_User, time=0.5, value=1}
        {arriveTangent=-2.6073, interpMode=RCIM_Cubic, leaveTangent=-2.6073, tangentMode=RCTM_User, time=0.83, value=0.55}
       ]}
     }}
    Scale Sprite Size Mode=::NewEnumerator3
   ) {enabled=false, module=ScaleSpriteSize002}
   &ScaleSpriteSize(Uniform Curve Sprite Scale=DataInterface() {propertyValues {Curve {keys [
        {arriveTangent=-0.0593, interpMode=RCIM_Cubic, leaveTangent=-0.0593, tangentMode=RCTM_User, time=0, value=1.3}
        {arriveTangent=-3.4885, interpMode=RCIM_Cubic, leaveTangent=-3.4885, tangentMode=RCTM_User, time=0.2, value=1}
       ]}}})
   &ScaleColor(
    Scale Alpha=&FloatFromCurve(FloatCurve=DataInterface() {propertyValues {Curve {keys [
         {arriveTangent=0.1417, interpMode=RCIM_Cubic, leaveTangent=0.1417, tangentMode=RCTM_User, time=0, value=0}
         {interpMode=RCIM_Cubic, tangentMode=RCTM_User, time=0.1, value=1}
         {arriveTangent=-1.4377, interpMode=RCIM_Cubic, leaveTangent=-1.4377, tangentMode=RCTM_User, time=0.58, value=0}
        ]}}})
    Scale RGB=&MakeVectorFromLinearColorRGB(Color=&ColorFromCurve(ColorCurve=DataInterface() {propertyValues {
        AlphaCurve {keys []}
        BlueCurve {keys [<at 0.1329 5>,<at 0.4464 1>]}
        GreenCurve {keys [<at 0.1329 2.2667>,<at 0.4464 1>]}
        RedCurve {keys [<at 0.1329 0.9>,<at 0.4464 1>]}
       }}))
    ScaleA=true
    ScaleRGB=true
   )
   /SetVariables(Particles.ColorDarkGlow:LinearColor=&MakeLinearColorFromVectorAndFloat(
     "Vector (RGB)"=<0 0 0>
     "Float (Alpha)"=&FloatFromCurve(FloatCurve=DataInterface() {propertyValues {Curve {keys [
          {arriveTangent=0.3631, interpMode=RCIM_Cubic, leaveTangent=0.3631, tangentMode=RCTM_User, time=0, value=0}
          {arriveTangent=-2.9613, interpMode=RCIM_Cubic, leaveTangent=-2.9613, tangentMode=RCTM_User, time=0.1, value=2}
          {arriveTangent=-0.223, interpMode=RCIM_Cubic, leaveTangent=-0.223, tangentMode=RCTM_User, time=1, value=0}
         ]}}})
    )) {isSetParameters=true}
   /SetVariables(Particles.ColorCorona:LinearColor=&MakeLinearColorFromVectorAndFloat(
     "Vector (RGB)"=<0.7 0.85 1>
     "Float (Alpha)"=&FloatFromCurve(
      FloatCurve=DataInterface() {propertyValues {Curve {keys [
          {arriveTangent=0.3631, interpMode=RCIM_Cubic, leaveTangent=0.3631, tangentMode=RCTM_User, time=0, value=0}
          {arriveTangent=-4.6901, interpMode=RCIM_Cubic, leaveTangent=-4.6901, tangentMode=RCTM_User, time=0.18, value=2}
          {arriveTangent=-3.1144, interpMode=RCIM_Cubic, leaveTangent=-3.1144, tangentMode=RCTM_User, time=0.5, value=0}
         ]}}}
      Scale Curve=0.2
     )
    )) {isSetParameters=true}
   &DynamicMaterialParameters(
    Index 1 Param 1=7
    Module.Index1Param0Write=true
    Module.Index1Param1Write=false
    Module.Index1Param2Write=false
    Module.Index1Param3Write=false
    Module.Index2Param0Write=false
    Module.Index2Param1Write=false
    Module.Index2Param2Write=false
    Module.Index2Param3Write=false
    Module.Index3Param0Write=false
    Module.Index3Param1Write=false
    Module.Index3Param2Write=false
    Module.Index3Param3Write=false
    Write Parameter Index 0=false
    Write Parameter Index 1=true
    Write Parameter Index 2=true
    Write Parameter Index 3=true
   )
  ]
  rendererClasses [&NiagaraSpriteRendererProperties]
  renderers [
   {rendererClass=&NiagaraSpriteRendererProperties}
   {rendererClass=&NiagaraSpriteRendererProperties, rendererIndex=1}
   {rendererClass=&NiagaraSpriteRendererProperties, rendererIndex=2}
  ]
 }
 Glow_Big {
  emitterUpdateScript [
   &EmitterState()
   &SpawnBurst_Instantaneous(Spawn Count=1, Use Spawn Probability=false)
  ]
  particleSpawnScript [
   &InitializeParticle(
    Color=<rgba 0.04 0.344 1 1>
    Lifetime=0.33
    Position Offset=<0 5 0>
    Sprite Size=&MakeVector2D(
     X=&SelectFloatFromArray(Direct Array Index=$User.FX_WeaponCaseSize, Float Selection Array=DataInterface() {propertyValues {FloatData [150,220,300]}})
     Y=170
    )
    Sprite Size Mode=::NewEnumerator4
    UsePositionOffset=false
    Write Material Random=true
   ) {module=InitializeParticle001}
   &SpriteFacingAndAlignment(Alignment Coordinate Space=::Local, Sprite Alignment=<1 0.25 0>, WriteAlignment=true, WriteFacing=false) {enabled=false}
   /SetVariables(Particles.Position=&AddVectorToPosition(Position=$System.FX_Position, Vector=<0 10 0>)) {isSetParameters=true}
   &CameraOffset(Camera Offset Amount=20)
  ]
  particleUpdateScript [
   &ParticleState()
   &ScaleSpriteSize(Uniform Curve Sprite Scale=DataInterface() {propertyValues {Curve {keys [{tangentMode=RCTM_User, time=0, value=0.5}, {tangentMode=RCTM_User, time=0.3, value=1}]}}}) {enabled=false}
   &ScaleColor(
    Scale Alpha=&FloatFromCurve(FloatCurve=DataInterface() {propertyValues {Curve {keys [
         {arriveTangent=0.0247, interpMode=RCIM_Cubic, leaveTangent=0.0247, tangentMode=RCTM_User, time=0.02, value=0}
         {arriveTangent=-0.7581, interpMode=RCIM_Cubic, leaveTangent=-0.7581, tangentMode=RCTM_User, time=0.13, value=1}
         {arriveTangent=0.0625, interpMode=RCIM_Cubic, leaveTangent=0.0625, tangentMode=RCTM_User, time=1, value=0}
        ]}}})
    Scale RGB=&VectorFromFloat(Value=&FloatFromCurve(FloatCurve=DataInterface() {propertyValues {Curve {keys [{interpMode=RCIM_Cubic, tangentMode=RCTM_User, time=0, value=5}, <at 0.27 1 RCIM_Cubic>]}}}))
    ScaleA=true
    ScaleRGB=true
   )
   /SetVariables(Particles.ColorLight:LinearColor=&MultiplyLinearColor(
     Linear Color=$Particles.Color
     Scale Factor=&MakeFloat4FromVectorAndFloat(
      "Vector (RGB)"=&VectorFromFloat(Value=6)
      "Float (Alpha)"=&FNTimeOfDayScalar(Night Scale=3, Morning Scale=4.4, Day Scale=5, Evening Scale=4)
     )
    )) {isSetParameters=true}
   &DynamicMaterialParameters(
    Index 0 Param 1=15
    Module.Index0Param0Write=true
    Module.Index0Param1Write=false
    Module.Index0Param2Write=false
    Module.Index0Param3Write=false
   )
  ]
  rendererClasses [&NiagaraSpriteRendererProperties, &NiagaraLightRendererProperties]
  renderers [
   {rendererClass=&NiagaraSpriteRendererProperties}
   {rendererClass=&NiagaraLightRendererProperties, rendererIndex=1}
   {rendererClass=&NiagaraSpriteRendererProperties, rendererIndex=2}
  ]
 }
 Smoke {
  emitterUpdateScript [
   &EmitterState()
   &SpawnBurst_Instantaneous(Spawn Count=&RandomRangeInt(Minimum=4, Maximum=7), Use Spawn Probability=false) {module=SpawnBurst_Instantaneous001}
  ]
  particleSpawnScript [
   &InitializeParticle(
    Color=<rgba 1 1 1 0.05>
    Lifetime=&UniformRangedFloat(Minimum=0.33, Maximum=0.44)
    Sprite Rotation Angle=&UniformRangedFloat(Minimum=0, Maximum=1)
    Sprite Rotation Mode=::NewEnumerator1
    Sprite Size Mode=::NewEnumerator3
    Uniform Sprite Size=&UniformRangedFloat(Minimum=33, Maximum=44)
    UsePositionOffset=false
    Write Material Random=true
   ) {module=InitializeParticle001}
   /SetVariables(Particles.InitSpriteRotRate=&UniformRangedFloat(Minimum=-1, Maximum=1)) {isSetParameters=true}
   &AccelerationForce(Acceleration=&UniformRangedVector(Minimum=<0 0 -25>, Maximum=<0 0 -75>))
   &SphereLocation(
    MaskSpawnGroup=false
    Non Uniform Scale=&MakeVector(
     X=&SelectFloatFromArray(Direct Array Index=$User.FX_WeaponCaseSize, Float Selection Array=DataInterface() {propertyValues {FloatData [2,3,5]}})
     Y=0.5
     Z=1
    )
    Override Seed=false
    Sphere Origin:NiagaraPosition=$System.FX_Position
    Sphere Radius=10
    Surface Only=true
    UseNonUniformScale=true
    UseOffset=false
    UseSphereOrientation=false
   ) {enabled=false}
   &ShapeLocation(
    Hemisphere Distribution=<0.5 1>
    MaskSpawnGroup=false
    Non Uniform Scale=&MakeVector(
     X=&SelectFloatFromArray(Direct Array Index=$User.FX_WeaponCaseSize, Float Selection Array=DataInterface() {propertyValues {FloatData [2,3,5]}})
     Y=0.5
     Z=1
    )
    Override Seed=false
    Shape Origin=$System.FX_Position
    Sphere Radius=14
    UseOwnerScale=true
   )
   &ApplyInitialForces()
   &CameraOffset(Camera Offset Amount=&RandomRangeFloat(Minimum=0, Maximum=15))
   &AddVelocity(Velocity=&RandomRangeVector(Minimum=<-20 0 -10>, Maximum=<20 30 10>))
  ]
  particleUpdateScript [
   &ParticleState()
   &SpriteRotationRate(Rotation Rate=&Multiply_Float(
     A=$Particles.InitSpriteRotRate
     B=&FloatFromCurve(
      FloatCurve=DataInterface() {propertyValues {Curve {keys [
          {arriveTangent=-1.9214, interpMode=RCIM_Cubic, leaveTangent=-1.9214, tangentMode=RCTM_User, time=-0.0046, value=1.0113}
          {arriveTangent=-0.2496, interpMode=RCIM_Cubic, leaveTangent=-0.2496, tangentMode=RCTM_User, time=0.9896, value=0.16}
         ]}}}
      CurveIndex=$Emitter.Age
     )
    ))
   &ScaleColor(
    Scale Alpha=&FloatFromCurve(FloatCurve=DataInterface() {propertyValues {Curve {keys [<at 0 0>,<at 0.2049 0.9765>,<at 0.6 0.9107>,<at 1 0>]}}})
    Scale RGB=&MakeVectorFromLinearColorRGB(Color=&ColorFromCurve(ColorCurve=DataInterface() {propertyValues {
        AlphaCurve {keys []}
        BlueCurve {keys [<at 0.2408 1 RCIM_Cubic>,<at 0.9209 0.7 RCIM_Cubic>]}
        GreenCurve {keys [<at 0.2408 0.6617 RCIM_Cubic>,<at 0.9209 0.7 RCIM_Cubic>]}
        RedCurve {keys [<at 0.2408 0.42 RCIM_Cubic>,<at 0.9209 0.7 RCIM_Cubic>]}
       }}))
    ScaleA=true
    ScaleRGB=true
   ) {module=ScaleColor001}
   &Drag(Drag=3, Use Linear Drag=true, Use Rotational Drag=false)
   &SolveForcesAndVelocity(Clamp Velocity=false, Limit Acceleration=false)
   &ScaleSpriteSize(Scale Factor=&Vector2DFromFloat(Value=&FloatFromCurve(FloatCurve=DataInterface() {propertyValues {Curve {keys [
          {arriveTangent=5.2354, interpMode=RCIM_Cubic, leaveTangent=5.2354, tangentMode=RCTM_User, time=-0.0046, value=0.25}
          <at 0.1439 1.4923>
          <at 1.0011 2.6 RCIM_Cubic>
         ]}}})))
   &DynamicMaterialParameters(
    Index 0 Param 1=&FloatFromCurve(FloatCurve=DataInterface() {propertyValues {Curve {keys [<at 0 0 RCIM_Cubic>,<at 1 0.2>]}}})
    Index 0 Param 2=&FloatFromCurve(FloatCurve=DataInterface() {propertyValues {Curve {keys [
         {arriveTangent=0.0453, interpMode=RCIM_Cubic, tangentMode=RCTM_Break, time=0, value=-1}
         {arriveTangent=0.0604, interpMode=RCIM_Cubic, leaveTangent=0.0604, tangentMode=RCTM_User, time=1, value=-1}
        ]}}})
    Index 0 Param 3=&FloatFromCurve(FloatCurve=DataInterface() {propertyValues {Curve {keys [{interpMode=RCIM_Cubic, leaveTangent=4.6779, time=0.2851, value=0}]}}})
    Index 1 Param 1=&FloatFromCurve(
     FloatCurve=DataInterface() {propertyValues {Curve {keys [<at 0 0>,<at 0.25 1>]}}}
     Scale Curve=&RandomRangeFloat(Minimum=2, Maximum=1)
    )
    Index 1 Param 2=&FloatFromCurve(
     FloatCurve=DataInterface() {propertyValues {Curve {keys [
         {interpMode=RCIM_Cubic, leaveTangent=3.7422, tangentMode=RCTM_Break, time=0, value=0.2}
         {interpMode=RCIM_Cubic, leaveTangent=0.9066, time=1, value=2}
        ]}}}
     Scale Curve=&RandomRangeFloat(Minimum=0.5, Maximum=1)
    )
    Index 1 Param 3=&FloatFromCurve(FloatCurve=DataInterface() {propertyValues {Curve {keys [
         {interpMode=RCIM_Cubic, leaveTangent=2.8445, tangentMode=RCTM_Break, time=0, value=0.2}
         {interpMode=RCIM_Cubic, leaveTangent=63.2978, time=0.25, value=1}
        ]}}})
    Index 1 Param 4=&FloatFromCurve(
     FloatCurve=DataInterface() {propertyValues {Curve {keys [<at 0 1>,<at 1 0.8>]}}}
     Scale Curve=&RandomRangeFloat(Minimum=0.5, Maximum=1)
    )
    Index 2 Param 1=&FloatFromCurve(FloatCurve=DataInterface() {propertyValues {Curve {keys [{interpMode=RCIM_Cubic, leaveTangent=3.2309, tangentMode=RCTM_Break, time=0.1, value=0}, <at 1 1>]}}})
    Index 2 Param 2=$Particles.MaterialRandom
    Index 2 Param 3=&FloatFromCurve(FloatCurve=DataInterface() {propertyValues {Curve {keys [
         {interpMode=RCIM_Cubic, tangentMode=RCTM_Break, time=0, value=1}
         {interpMode=RCIM_Cubic, leaveTangent=11.1233, tangentMode=RCTM_Break, time=1, value=2}
        ]}}})
    Index 2 Param 4=0.75
    Module.Param0WriteEnabled=true
    Module.Param0WriteEnabled1=true
    Module.Param0WriteEnabled2=true
    Module.Param1WriteEnabled=false
    Module.Param1WriteEnabled1=true
    Module.Param1WriteEnabled2=true
    Module.Param2WriteEnabled=true
    Module.Param2WriteEnabled1=true
    Module.Param2WriteEnabled2=true
    Module.Param3WriteEnabled=true
    Module.Param3WriteEnabled1=true
    Module.Param3WriteEnabled2=true
    Write Parameter Index 1=true
    Write Parameter Index 2=true
   )
  ]
  rendererClasses [&NiagaraSpriteRendererProperties]
  renderers [{rendererClass=&NiagaraSpriteRendererProperties}]
 }
 Sparks {
  emitterUpdateScript [
   &EmitterState()
   &SpawnBurst_Instantaneous(Spawn Count=12, Use Spawn Probability=false, UseLoopCountLimit=false)
  ]
  particleSpawnScript [
   &InitializeParticle(
    Color Maximum=<rgba 0.56 3.66 8 0.35>
    Color Minimum=<rgba 0.56 4.8533 7 0.2>
    Color Mode=::NewEnumerator2
    Lifetime Max=0.66
    Lifetime Min=0.22
    Lifetime Mode=::NewEnumerator1
    Sprite Size=&Multiply_Vector2D_ByFloat(Float=&RandomRangeFloat(Minimum=0.8, Maximum=2.5), Vector2D=<1 1>)
    Sprite Size Mode=::NewEnumerator4
    UsePositionOffset=false
    Write Material Random=true
   )
   &ShapeLocation(
    Hemisphere Distribution=<0.5 1>
    MaskSpawnGroup=false
    Non Uniform Scale=&MakeVector(
     X=&SelectFloatFromArray(Direct Array Index=$User.FX_WeaponCaseSize, Float Selection Array=DataInterface() {propertyValues {FloatData [1,1.5,2]}})
     Y=0.5
     Z=1
    )
    Override Seed=false
    Shape Origin=$System.FX_Position
    Sphere Radius=42
    UseOwnerScale=true
   )
   &AddVelocity(
    Module.Use Falloff Distance=false
    Origin Offset=<0 -35 -6>
    Override Seed=false
    Velocity Mode=::NewEnumerator1
    Velocity Origin=$System.FX_Position
    Velocity Speed=&RandomRangeFloat(Minimum=80, Maximum=100)
   )
  ]
  particleUpdateScript [
   &ParticleState()
   &ScaleSpriteSize(
    Non-Uniform Curve Sprite Scale=DataInterface() {propertyValues {
      XCurve {keys [{interpMode=RCIM_Cubic, tangentMode=RCTM_User, time=0.2, value=0.7}, {interpMode=RCIM_Cubic, tangentMode=RCTM_User, time=0.4, value=1}]}
      YCurve {keys [{interpMode=RCIM_Cubic, tangentMode=RCTM_User, time=0.1, value=4}, {interpMode=RCIM_Cubic, tangentMode=RCTM_User, time=0.4, value=1}]}
     }}
    Scale Sprite Size Mode=::NewEnumerator3
   )
   &Drag(Use Linear Drag=true, Use Rotational Drag=true)
   &GravityForce(Gravity=<0 0 -120>)
   &SolveForcesAndVelocity(Clamp Velocity=false, Limit Acceleration=false)
   &ScaleColor(
    Scale Alpha=&FloatFromCurve(
     FloatCurve=DataInterface() {propertyValues {Curve {keys [
         {arriveTangent=18.2736, interpMode=RCIM_Cubic, leaveTangent=18.2736, tangentMode=RCTM_User, time=0, value=0}
         {arriveTangent=-1.0125, interpMode=RCIM_Cubic, leaveTangent=-1.0125, tangentMode=RCTM_User, time=0.1, value=1}
         {interpMode=RCIM_Cubic, tangentMode=RCTM_User, time=1, value=0}
        ]}}}
     Scale Curve=0.5
    )
    Scale RGB=&VectorFromFloat(Value=&FloatFromCurve(FloatCurve=DataInterface() {propertyValues {Curve {keys [{interpMode=RCIM_Cubic, tangentMode=RCTM_User, time=0, value=5}, {interpMode=RCIM_Cubic, tangentMode=RCTM_User, time=0.3, value=1}]}}}))
    ScaleA=true
    ScaleRGB=true
   )
  ]
  rendererClasses [&NiagaraSpriteRendererProperties]
  renderers [{rendererClass=&NiagaraSpriteRendererProperties}]
 }
}
summary {
 systemName=NS_WeaponCache_Interact
 emitters [
  {emitterName=Flares, rendererClasses [&NiagaraSpriteRendererProperties]}
  {emitterName=Glow_Big, rendererClasses [&NiagaraSpriteRendererProperties, &NiagaraLightRendererProperties]}
  {emitterName=Sparks, rendererClasses [&NiagaraSpriteRendererProperties]}
  {emitterName=Smoke, rendererClasses [&NiagaraSpriteRendererProperties]}
  {emitterName=Aura, rendererClasses [&NiagaraMeshRendererProperties]}
 ]
 userVariables [User.FX_WeaponCaseSize {type=NiagaraInt32, defaultValue=1}]
}
