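# Translucent unlit, two-sided material for the Paprika weapon case padding that fakes a light glow over nearby surfaces without a real light.
# A UV-built shape mask fades out with pixel depth and switches off through LightOff custom primitive data.
# Emissive multiplies tinted scene color by scene-depth attenuation, with a simpler mobile path and inverse eye-adaptation compensation.
# A camera-relative horizontal basis yields a vertex offset that turns the mesh like a yaw-only billboard.
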Material /PaprikaWeaponCases/Assets/Materials/M_PaprikaWeaponCase_Padding_Light
graph {
 properties {blendMode=BLEND_Translucent, pixelDepthOffsetMode=PDOM_Legacy, shadingModel=MSM_Unlit, twoSided=true, usedShadingModels=Unlit}
 nodes [
  Constant() {r=1}
  Constant()
  n1 = MaterialFunctionCall(Float2=TextureCoordinate()) {materialFunction=&BreakOutFloat2Components}
  n2 = Power(Base=Abs(None=ConstantBiasScale(None=@n1.R) {bias=-0.5, scale=2}))
  n3 = MaterialFunctionCall(Float2=AppendVector(A=OneMinus(None=@n2), B=@n1.G)) {materialFunction=&BreakOutFloat2Components}
  n4 = Add(A=@n3.R) {constB=0.8}
  n5 = Power(Base=Abs(None=ConstantBiasScale(None=@n3.G) {bias=-0.5, scale=2})) {constExponent=1}
  n6 = OneMinus(None=SphereMask(A=Multiply(A=@n4) {constB=-0.8}, B=OneMinus(None=@n5)) {attenuationRadius=1.5, hardnessPercent=70})
  n7 = Multiply(A=Power(Base=Multiply(A=@n3.R)) {constExponent=3}) {constB=0.6}
  n8 = NamedRerouteDeclaration(None=Multiply(A=Subtract(A=@n6, B=@n7))) {name=Shape, nodeColor=<rgba 0.4297 1 0 1>}
  n9 = PixelDepth()
  n10 = ScalarParameter() {defaultValue=1000, parameterName=OpacityLerpStart, sliderMax=200}
  n11 = Saturate(None=Divide(A=Subtract(A=@n9.R, B=@n10), B=ScalarParameter() {defaultValue=500, parameterName=OpacityLerpDistance, sliderMax=1000}))
  n12 = ScalarParameter() {bUseCustomPrimitiveData=true, parameterName=LightOff}
  Multiply(
   A=Multiply(A=LinearInterpolate(A=@n8, Alpha=@n11) {constAlpha=0, constB=0}, B=OneMinus(None=@n12))
   B=LinearInterpolate(
    A=ScalarParameter() {defaultValue=0.12, parameterName=OpcaityMult2}
    B=ScalarParameter() {defaultValue=0.2, parameterName=OpcaityMult}
    Alpha=NamedRerouteUsage()
   )
  )
  n13 = ScalarParameter() {defaultValue=160, parameterName=LightStrength}
  n14 = NamedRerouteUsage()
  n15 = Subtract(A=SceneDepth().R, B=PixelDepth().R)
  n16 = ScalarParameter() {defaultValue=50, parameterName=LightAttenuation}
  n17 = DepthFade(
   Opacity=OneMinus(None=Divide(A=@n15, B=@n16))
   FadeDistance=LinearInterpolate(A=ScalarParameter() {defaultValue=20, parameterName=DepthFadeDistance}, Alpha=NamedRerouteUsage()) {constA=10, constB=30}
  )
  n18 = Multiply(
   A=VectorParameter() {parameterName=LightColor, defaultValue=<rgba 0.5396 0.6182 0.8542 1>}.RGB
   B=NamedRerouteDeclaration(None=Saturate(None=@n17)) {name=Opacity, nodeColor=<rgba 1 0 0.8438 1>}
  )
  n19 = ComponentMask(None=SceneColor!(UVs).RGB) {b=true, g=true, r=true}
  n20 = SceneTexture() {sceneTextureId=PPI_DiffuseColor}
  n21 = ComponentMask(None=ActorPositionWS()) {b=true}
  n22 = LinearInterpolate(
   A=LinearInterpolate(A=@n19, B=ComponentMask(None=@n20.Color) {b=true, g=true, r=true}) {constAlpha=0}
   B=@n19
   Alpha=NamedRerouteDeclaration(None=Step(X=@n21) {constX=2500, constY=2500}) {name=PositionMask, nodeColor=<rgba 0.9688 1 0 1>}
  )
  n23 = EyeAdaptation()
  n24 = Multiply(
   A=Multiply(A=LinearInterpolate(B=@n13, Alpha=@n14), B=@n18)
   B=LinearInterpolate(A=@n22, B=Multiply(A=@n22, B=@n23.EyeAdaptation)) {constAlpha=1}
  )
  n25 = Multiply(A=LinearInterpolate(B=ScalarParameter() {defaultValue=0.005, parameterName=Mobile-LightStrength}, Alpha=@n14), B=@n18)
  NamedRerouteDeclaration(None=EyeAdaptationInverse(
    LightValueInput=ShadingPathSwitch(Default=@n24, Mobile=@n25)
    AlphaInput=ScalarParameter() {defaultValue=1, parameterName=EyeAdaptionInverse}
   ).EyeAdaptationInverse) {name=Emissive, nodeColor=<rgba 0 0.625 1 1>}
  NamedRerouteUsage()
  NamedRerouteUsage()
  n26 = MaterialFunctionCall(Float3=MaterialFunctionCall() {materialFunction=&LocalPosition}."Local Position (Excluding Offsets)") {materialFunction=&BreakOutFloat3Components}
  n27 = MaterialFunctionCall(X=@n26.G, Y=@n26.G, Z=@n26.G) {materialFunction=&MakeFloat3}
  n28 = ObjectPositionWS()
  n29 = CameraPositionWS()
  n30 = Add(A=Multiply(A=Subtract(A=@n28, B=@n29), B=ScalarParameter() {defaultValue=-500, parameterName=Param_9}) {constB=500})
  n31 = MaterialFunctionCall(Float2=@n30) {materialFunction=&BreakOutFloat2Components}
  n32 = Distance(A=@n31.R, B=Constant())
  n33 = ComponentMask(None=@n30) {g=true, r=true}
  n34 = AppendVector(A=Constant(), B=@n31.G)
  n35 = AppendVector(A=Normalize(VectorInput=If(A=@n32, "A > B"=@n33, "A == B"=@n34, "A < B"=@n34) {constB=7500}), B=Constant())
  n36 = Constant3Vector() {constant=<rgba 0 0 1 0>}
  n37 = CrossProduct(A=@n35, B=@n36)
  Subtract(
   A=TransformPosition(None=MaterialFunctionCall!(VectorToTransform=@n27.Result, BasisX=@n35, BasisY=@n37, BasisZ=@n36) {materialFunction=/Engine/Functions/Engine_MaterialFunctions02/Math/InverseTransformMatrix}.Transformed Vector 3) {transformType=TRANSFORMPOSSOURCE_World}
   B=WorldPosition() {worldPositionShaderOffset=WPT_ExcludeAllShaderOffsets}.XYZ
  )
 ]
}
