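# Default Lit velvet material for weapon case padding, usable on skeletal and Nanite meshes.
# Diffuse blends grunge-varied reds and a Fresnel highlight masked by baked AO, which also drives roughness and specular.
# A flattened padding normal and velvet height texture feed FuzzyShading, while local position provides a UV set.
# An eye-adaptation-compensated glow sums faint diffuse and a time-of-day-scaled term, alongside a UV-gradient fake light.
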Material /PaprikaWeaponCases/Assets/Materials/M_PaprikaWeaponCase_Padding_Velvet
graph {
 properties {
  bUsedWithMeshDeformer=true
  bUsedWithNanite=true
  bUsedWithSkeletalMesh=true
  pixelDepthOffsetMode=PDOM_Legacy
  usedShadingModels=Default Lit
 }
 nodes [
  Constant() {r=1}
  n1 = Multiply(A=NamedRerouteUsage()) {constB=0.15}
  n2 = LinearInterpolate(
   A=VectorParameter() {parameterName=Color_Curvature, defaultValue=<rgba 0.1667 0 0.0141 1>}.RGB
   B=LinearInterpolate(
    A=VectorParameter() {defaultValue=<rgba 0.4531 0.0085 0.0356 1>}.RGB
    B=VectorParameter() {parameterName=Param4, defaultValue=<rgba 0.3854 0.01 0.0334 1>}.RGB
    Alpha=TextureSampleParameter2D(UVs=@n1) {parameterName=Grunge, samplerType=SAMPLERTYPE_Masks, texture=/PaprikaWeaponCases/Assets/Textures/T_Velvet_GrungeMask}.RGB
   ) {constAlpha=0}
   Alpha=NamedRerouteUsage()
  )
  n3 = VectorParameter() {parameterName=Param_1, defaultValue=<rgba 1 0.0312 0.0312 1>}
  n4 = TextureSampleParameter2D(UVs=NamedRerouteUsage()) {parameterName=AO, samplerType=SAMPLERTYPE_LinearColor, texture=/PaprikaWeaponCases/Assets/Textures/CasePadding_ambientocclusion}
  n5 = Constant() {r=0.9}
  n6 = NamedRerouteDeclaration(None=Multiply(A=NamedRerouteUsage()) {constB=0.3}) {name=UV2, nodeColor=<rgba 0.6875 1 0 1>}
  n7 = NamedRerouteDeclaration(None=MaterialFunctionCall(
    Normal=TextureSample(UVs=@n6) {samplerType=SAMPLERTYPE_Normal, texture=/PaprikaWeaponCases/Assets/Textures/CasePadding_normal}.RGB
    Flatness=Constant() {r=0.7}
   ) {materialFunction=&FlattenNormal}.Result) {name=Normal, nodeColor=<rgba 0.2188 0 1 1>}
  n8 = Fresnel(Normal=Transform(None=@n7)) {exponent=1}
  n9 = Multiply(A=Power(Base=SphereMask(A=@n4.B, B=@n5) {attenuationRadius=0.1, hardnessPercent=0}) {constExponent=1}, B=Power(Base=@n8) {constExponent=5})
  n10 = Multiply(
   A=NamedRerouteDeclaration(None=LinearInterpolate(A=@n2, B=@n3.RGB, Alpha=@n9)) {name=Diffuse, nodeColor=<rgba 0.9922 0 1 1>}
   B=ScalarParameter() {defaultValue=0.02, parameterName=0.01}
  )
  n11 = ScalarParameter() {defaultValue=0.02, parameterName=FuzzNight}
  n12 = ScalarParameter() {defaultValue=0.04, parameterName=Fuzz_Day}
  n13 = Multiply(
   A=Multiply(
    A=NamedRerouteUsage()
    B=MaterialFunctionCall(Value_Morning=@n11, Value_Day=@n12, Value_Evening=@n11, Value_Night=@n11) {materialFunction=&Scalar_Time_of_Day}.Result
   ) {constB=0.1}
   B=NamedRerouteUsage()
  )
  EyeAdaptationInverse(
   LightValueInput=Add(A=Add(A=@n10, B=@n13), B=NamedRerouteUsage()) {constB=0}
   AlphaInput=ScalarParameter() {defaultValue=1, parameterName=EyeAdaptionInverse}
  )
  n14 = TextureSampleParameter2D(UVs=Multiply(A=NamedRerouteUsage(), B=ScalarParameter() {defaultValue=5, parameterName=Param_2}) {constB=10}) {parameterName=Height, samplerType=SAMPLERTYPE_Masks, texture=/PaprikaWeaponCases/Assets/Textures/T_Velvet_Height}
  n15 = ScalarParameter() {defaultValue=1, parameterName=Fuzz_Power}
  Multiply(A=@n14.R, B=@n15)
  n16 = ScalarParameter() {defaultValue=1, parameterName=Fuzz_CoreDarkness}
  Multiply(A=@n14.R, B=@n16)
  n17 = TextureSampleParameter2D(UVs=NamedRerouteUsage()) {parameterName=AO, samplerType=SAMPLERTYPE_LinearColor, texture=&CasePadding_ambientocclusion}
  Multiply(A=@n14.R, B=@n17.R)
  n18 = NamedRerouteDeclaration(None=Saturate(None=Add(A=@n17.R) {constB=0.12})) {name=Diamonds, nodeColor=<rgba 0 0.4375 1 1>}
  NamedRerouteDeclaration(None=@n18) {name=AO, nodeColor=<rgba 0.5469 1 0 1>}
  n19 = VectorParameter() {parameterName=FakeLight, defaultValue=<rgba 0 0.2091 1 1>}
  n20 = MaterialFunctionCall(Float2=TextureCoordinate()) {materialFunction=&BreakOutFloat2Components}
  n21 = Abs(None=ConstantBiasScale(None=@n20.R) {bias=-0.5, scale=2})
  n22 = ConstantBiasScale(None=@n20.G) {bias=-0.5, scale=2}
  n23 = Multiply(A=Power(Base=SmoothStep(Value=@n21)) {constExponent=1}, B=OneMinus(None=Abs(None=@n22)))
  NamedRerouteDeclaration(None=Multiply(A=Multiply(A=@n19.RGB, B=@n23), B=ScalarParameter() {defaultValue=0.01, parameterName=FakeLightIntensity}) {constB=0.2}) {name=FakeLight, nodeColor=<rgba 1 0 0.0937 1>}
  n24 = WorldPosition()
  n25 = MaterialFunctionCall(Float3=Divide(A=Transform(None=@n24.XYZ) {transformSourceType=TRANSFORMSOURCE_World, transformType=TRANSFORM_Local}, B=ScalarParameter() {defaultValue=40}) {constB=100}) {materialFunction=&BreakOutFloat3Components}
  NamedRerouteDeclaration(None=AppendVector(A=@n25.R, B=@n25.R)) {name=UV, nodeColor=<rgba 1 0.5859 0 1>}
  NamedRerouteDeclaration(None=MaterialFunctionCall!(
    BaseColor=NamedRerouteUsage()
    Normal=@n7
    CoreDarkness=@n16
    Power=@n15
    EdgeBrightness=ScalarParameter() {defaultValue=3, parameterName=Fuzz_EdgeBrightness}
   ) {materialFunction=/Engine/Functions/Engine_MaterialFunctions01/Shading/FuzzyShading}.Result) {name=Fuzz, nodeColor=<rgba 0.9453 0 1 1>}
  n26 = Power(Base=@n18) {constExponent=0.6}
  NamedRerouteDeclaration(None=LinearInterpolate(
    A=ScalarParameter() {defaultValue=1, parameterName=Roughness_Max}
    B=ScalarParameter() {defaultValue=0.4, parameterName=Roughness_Min}
    Alpha=@n26
   )) {name=Roughness, nodeColor=<rgba 0.1484 0 1 1>}
  NamedRerouteDeclaration(None=LinearInterpolate(A=ScalarParameter() {parameterName=Specular_Max}, B=ScalarParameter() {defaultValue=0.25, parameterName=Specular_Min}, Alpha=@n26)) {name=Specular, nodeColor=<rgba 0 1 0.8359 1>}
  NamedRerouteUsage()
  NamedRerouteUsage()
  NamedRerouteUsage()
  NamedRerouteUsage()
  Power(Base=NamedRerouteUsage())
  Saturate(None=@n8)
  ScalarParameter() {parameterName=Param_1}
  VectorParameter() {parameterName=Param_2, defaultValue=<rgba 0.9635 0 0 1>}
  VertexNormalWS()
 ]
}
