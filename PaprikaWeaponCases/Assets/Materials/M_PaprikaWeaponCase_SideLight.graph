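# Unlit material for the side light of the Paprika weapon case, with a glow that can be switched off per primitive.
# A LightOff scalar read from Custom Primitive Data is inverted and clamped to a small floor, so the light dims to a faint trace instead of black.
# That factor scales a Fresnel term and tints it with the LightColor parameter, an over-bright cyan by default.
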Material /PaprikaWeaponCases/Assets/Materials/M_PaprikaWeaponCase_SideLight
graph {
 properties {pixelDepthOffsetMode=PDOM_Legacy, shadingModel=MSM_Unlit, usedShadingModels=Unlit}
 nodes [
  n1 = OneMinus(None=ScalarParameter() {bUseCustomPrimitiveData=true, parameterName=LightOff})
  n2 = Fresnel() {exponent=1}
  Multiply(A=VectorParameter() {parameterName=LightColor, defaultValue=<rgba 0 2 3 1>}.RGB, B=Multiply(A=Clamp(None=@n1) {minDefault=0.01}, B=@n2))
  Power(Base=@n2) {constExponent=1}
 ]
}
